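# Kor boss challenge HUD widget, a child of WBP_Challenge_WidgetBase showing the boss name above a Kor-specific progress bar.
# Overrides forward challenge manager state to the bar for style, fill, stage markers and shake.
# Phase visuals use the bar's fill-finished delegate, with a timer on the last phase.
# Outro completion switches to boss-bar elements or, after challenge completion, collapses the widget; intro and outro animations have no bindings.
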WidgetBlueprint /KorBoss/UI/WBP_Kor_Challenge_Widget
animations [
 Anim_Intro {binding_count=0, length_seconds=5, object=&self:Anim_Intro, sequence {bindings []}}
 Anim_Outro {binding_count=0, length_seconds=5, object=&self:Anim_Outro, sequence {bindings []}}
]
graphs [
 BeginChallengeInitialSetup {nodes [
   BeginChallengeInitialSetup() {is=title K2Node_FunctionEntry}
   > InitialSetupOfChallengeVariables(self:WBP Challenge Widget Base Object Reference)
   > InitialSetupOfTextsAndStageMarkers!(
    execute
    self:WBP Challenge Widget Base Object Reference
    Challenge_StageIndicatorText:WBP Challenge Stage Indicator Text Object Reference
    Challenge Stage Markers Container:WBP Challenge Stage Markers Container Object Reference
    Challenge Progress Bar:WBP Challenge Progress Bar Object Reference=Get WBP_Kor_Challenge_ProgressBar() -> (WBP_Kor_Challenge_ProgressBar:WBP Kor Challenge Progress Bar Object Reference)
   ) -> (then)
   > SetVisibility(InVisibility=HitTestInvisible)
   > ForceUpdateProgressBarVisibility!(execute, self:WBP Challenge Widget Base Object Reference, Visible:ESlateVisibility Enum=HitTestInvisible) -> (then)
   > PlayIntroAnimation(self:WBP Challenge Widget Base Object Reference)
  ]}
 EventGraph {nodes [
   Event CheckBossReadyForFight() {is=title K2Node_Event}
   n1 = Get ChallengeManagerRef() -> (ChallengeManagerRef:Challenge Manager Object Reference)
   > Is Valid(InputObject=@n1)
   > n2 = Branch(Condition=Get BossIsReadyForFinalFight!(self:Challenge Manager Object Reference=@n1) -> (BossIsReadyForFinalFight:bool))
   > UpdateChallengeInformation!(execute, self:WBP Challenge Widget Base Object Reference, WithAnimation:bool=false) -> (then)
   > HandleBossProgressUpdatedInternal!(execute, self:WBP Challenge Widget Base Object Reference, AnimateProgressBar:bool=false) -> (then)
   UpdateStageMarkersState(execute=@n2.else, self:WBP Challenge Widget Base Object Reference)
   > PlayShakeAnimation!(execute, self:WBP Challenge Progress Bar Object Reference=Get WBP_Kor_Challenge_ProgressBar(), AnimationDelay:double=0) -> (then)
   Event Construct()
   > Parent: Construct()
   > ResetProgressBarFillValue!(
    execute
    self:WBP Challenge Widget Base Object Reference
    Challenge Progress Bar:WBP Challenge Progress Bar Object Reference=Get WBP_Kor_Challenge_ProgressBar()
   ) -> (then)
   Event OnChallengeCompletedVisualUpdates() {is=title K2Node_Event}
   > Delay(Duration=Get TimeToHideWidgetOnceFinished() -> (TimeToHideWidgetOnceFinished:double))
   > n3 = Play Animation with Finished event!(
    execute:Exec
    Widget:User Widget Object Reference=Self-Reference()
    InAnimation:Widget Animation Object Reference=Get Anim_Outro() -> (Anim_Outro:Widget Animation Object Reference)
    StartAtTime:float=0
    NumLoopsToPlay:int=1
    PlayMode:EUMGSequencePlayMode Enum
    PlaybackSpeed:float=1
   ) -> (then:Exec, Result:Widget Animation Handle Structure, Finished:Exec) {is=title K2Node_PlayAnimation2}
   SetVisibility(execute=@n3.Finished, InVisibility=Collapsed)
   Event PlayIntroAnimation() {is=title K2Node_Event}
   > Play Animation with Finished event(
    Widget=Self-Reference()
    InAnimation=Get Anim_Intro() -> (Anim_Intro:Widget Animation Object Reference)
    StartAtTime=0
    NumLoopsToPlay=1
    PlaybackSpeed=1
   )
   Event PlayOutroAnimation() {is=title K2Node_Event}
   > n4 = Play Animation with Finished event(Widget=Self-Reference(), InAnimation=Get Anim_Outro(), StartAtTime=0, NumLoopsToPlay=1, PlaybackSpeed=1)
   SetupBossBarElements!(execute=@n4.Finished, self:WBP Challenge Progress Bar Object Reference=Get WBP_Kor_Challenge_ProgressBar(), IsInBossStage:bool=true) -> (then)
  ]}
 ForceUpdateProgressBarVisibility {nodes [
   n1 = ForceUpdateProgressBarVisibility!() -> (then, Visible:ESlateVisibility Enum=Visible) {is=title K2Node_FunctionEntry}
   > Parent: ForceUpdateProgressBarVisibility!(execute, Visible:ESlateVisibility Enum=@n1) -> (then) {is=title K2Node_CallParentFunction}
   > SetVisibility(self=Get WBP_Kor_Challenge_ProgressBar(), InVisibility=@n1)
  ]}
 PhaseCompletedVisualUpdates {nodes [
   PhaseCompletedVisualUpdates() {is=title K2Node_FunctionEntry}
   > Unbind Event from On Bar Fill Animation Finished(self:WBP Challenge Progress Bar Object Reference=Get WBP_Kor_Challenge_ProgressBar(), Delegate=Create Event()) {is=title K2Node_RemoveDelegate}
   > UpdateStageMarkersState()
   > PlayShakeAnimation(self=Get WBP_Kor_Challenge_ProgressBar(), AnimationDelay=0)
   n1 = IsLastPhase!(self:Challenge Manager Object Reference=Get ChallengeManagerRef(), PhaseIndex:int=Get CurrentPhase() -> (CurrentPhase:int=0)) -> (Value:bool)
   > n2 = Branch(Condition=@n1)
   > Set Timer by Event(Delegate=Create Event(), Time=4.25, bLooping=false)
   SetProgressBarStyle!(execute=@n2.else, self:WBP Challenge Progress Bar Object Reference=Get WBP_Kor_Challenge_ProgressBar(), HasReachedBossStage:bool=@n1) -> (then)
  ]}
 UpdateChallengeCompleted {nodes [
   n1 = UpdateChallengeCompleted!() -> (then, IsChallengeCompleted:bool) {is=title K2Node_FunctionEntry}
   > n2 = Branch(Condition=@n1)
   > OnChallengeCompletedVisualUpdates(self:WBP Challenge Widget Base Object Reference)
   > SetProgressWithAnimation!(execute, self:WBP Challenge Progress Bar Object Reference=Get WBP_Kor_Challenge_ProgressBar(), ProgressEnd:double=0) -> (then)
   n3 = Get WBP_Kor_Challenge_ProgressBar()
   WaitForBarAnimationToFinish(execute=@n2.else, self:WBP Challenge Progress Bar Object Reference=@n3)
   > Bind Event to On Bar Fill Animation Finished(self:WBP Challenge Progress Bar Object Reference=@n3, Delegate=Create Event()) {is=title K2Node_AddDelegate}
  ]}
 UpdateChallengeInformation {nodes [
   UpdateChallengeInformation!() -> (then, WithAnimation:bool) {is=title K2Node_FunctionEntry}
   > SetProgressBarStyle(self=Get WBP_Kor_Challenge_ProgressBar(), HasReachedBossStage=IsOnBossStage!() -> (IsOnBossStage:bool))
   > PlayIntroAnimation()
  ]}
 UpdateProgressBarFill {nodes [
   UpdateProgressBarFill!() -> (then, With Animation:bool) {is=title K2Node_FunctionEntry}
   > SetProgressDirectly!(
    execute
    self:WBP Challenge Progress Bar Object Reference=Get WBP_Kor_Challenge_ProgressBar()
    Progress:double=Get Boss Progress() -> (Boss Progress:double)
   ) -> (then)
  ]}
]
variables []
widgets {
 info {
  inheritedWidgetCount=1
  parentClass=/AIProgressBar/UI/WBPs/WBP_Challenge_WidgetBase.WBP_Challenge_WidgetBase_C
  rootWidgetClass=&Overlay
  widgetCount=6
 }
 widgets [
  Overlay_Root {widgetClassPath=&Overlay, widget=Overlay_Root}
  VerticalBox_26 {
   widgetClassPath=&VerticalBox
   parent=Overlay_Root
   widget=VerticalBox_26
   slot {class=OverlaySlot, properties {horizontalAlignment=HAlign_Fill, verticalAlignment=VAlign_Center}}
  }
  Overlay_BossName {
   widgetClassPath=&Overlay
   parent=VerticalBox_26
   widget=Overlay_BossName
   slot {class=VerticalBoxSlot, properties {horizontalAlignment=HAlign_Center, verticalAlignment=VAlign_Center, padding=<ltrb 0 0 0 4>}}
  }
  Text_BossName {
   widgetClassPath=&CommonTextBlock
   parent=Overlay_BossName
   widget=Text_BossName
   properties {
    justification=Center
    text=Kor
    textTransformPolicy=ToUpper
    font {
     bForceMonospaced=false
     bMaterialIsStencil=false
     fontMaterial=null
     fontObject=&Burbank
     letterSpacing=36
     monospacedWidth=0.552
     size=24
     skewAmount=0
     typefaceFontName=Burbank-Black
     outlineSettings {
      bApplyOutlineToDropShadows=false
      bMiteredCorners=false
      bSeparateFillAlpha=false
      outlineMaterial=null
      outlineSize=0
      outlineColor=<rgba 0.0781 0.0781 0.0781 0.235>
     }
    }
    shadowColorAndOpacity=<rgba 0 0 0 0.56>
    shadowOffset=<1.5 1.5>
   }
   slot {class=OverlaySlot, properties {horizontalAlignment=HAlign_Fill, verticalAlignment=VAlign_Fill, padding=<ltrb 4 2 4 2>}}
  }
  WBP_Kor_Challenge_ProgressBar {
   widgetClassPath=&/WBP_Kor_Challenge_ProgressBar
   bIsVariable=true
   parent=VerticalBox_26
   widget=WBP_Kor_Challenge_ProgressBar
   properties {bOverride_Cursor=true}
   slot {class=VerticalBoxSlot, properties {horizontalAlignment=HAlign_Center}}
  }
  TutorialHighlight {widgetClassPath=&TutorialHighlight}
 ]
}
